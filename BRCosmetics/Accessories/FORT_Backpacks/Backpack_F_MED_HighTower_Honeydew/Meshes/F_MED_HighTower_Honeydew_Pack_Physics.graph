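# Physics asset for the HighTower Honeydew backpack cosmetic, giving its pack and two dangling chains collision bodies and joints for physics motion.
# A root box joins the pack capsule through an angularly free constraint, and the pack carries a chain base capsule.
# Left and right five-segment capsule chains branch from that base with tight per-link angle limits, ending in larger capsules at the tips.
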PhysicsAsset /BRCosmetics/Accessories/FORT_Backpacks/Backpack_F_MED_HighTower_Honeydew/Meshes/F_MED_HighTower_Honeydew_Pack_Physics
bodies {
 dyn_chains_base [{shapeType=Capsule, length=1, radius=1}]
 dyn_chains_l_01 [{shapeType=Capsule, length=1, radius=1, center=<1 0 0>, rotation=<rot 90 0 0>}]
 dyn_chains_l_02 [{shapeType=Capsule, length=1, radius=1, center=<1 0 0>, rotation=<rot 90 0 0>}]
 dyn_chains_l_03 [{shapeType=Capsule, length=1, radius=1, center=<1 0 0>, rotation=<rot 90 0 0>}]
 dyn_chains_l_04 [{shapeType=Capsule, length=2.25, radius=2, center=<3.6943 0 0>, rotation=<rot 90 0 0>}]
 dyn_chains_l_05 [{shapeType=Capsule, length=4.25, radius=2.8, center=<4.7557 0 0>, rotation=<rot 0 0 90>}]
 dyn_chains_r_01 [{shapeType=Capsule, length=1, radius=1, center=<1 0 0>, rotation=<rot 90 0 0>}]
 dyn_chains_r_02 [{shapeType=Capsule, length=1, radius=1, center=<1 0 0>, rotation=<rot 90 0 0>}]
 dyn_chains_r_03 [{shapeType=Capsule, length=1, radius=1, center=<1 0 0>, rotation=<rot 90 0 0>}]
 dyn_chains_r_04 [{shapeType=Capsule, length=2.25, radius=2, center=<3.6943 0 0>, rotation=<rot 90 0 0>}]
 dyn_chains_r_05 [{shapeType=Capsule, length=4.25, radius=2.8, center=<4.7557 0 0>, rotation=<rot 0 0 90>}]
 pack [{shapeType=Capsule, length=28.0561, radius=1.8373, center=<-0.0223 -4.3873 2.9389>}]
 root [{shapeType=Box, extentX=53.5, extentY=2.5, extentZ=38, center=<0 2.2462 0>}]
}
constraints [
 {bone1Name=dyn_chains_base, bone2Name=pack, swing1LimitDegrees=45, swing2LimitDegrees=45, twistLimitDegrees=45}
 {bone1Name=dyn_chains_r_01, bone2Name=dyn_chains_base, swing1LimitDegrees=10, swing2LimitDegrees=10, twistLimitDegrees=10}
 {bone1Name=dyn_chains_r_02, bone2Name=dyn_chains_r_01, swing1LimitDegrees=10, swing2LimitDegrees=10, twistLimitDegrees=10}
 {bone1Name=dyn_chains_r_03, bone2Name=dyn_chains_r_02, swing1LimitDegrees=10, swing2LimitDegrees=10, twistLimitDegrees=10}
 {bone1Name=dyn_chains_r_04, bone2Name=dyn_chains_r_03, swing1LimitDegrees=10, swing2LimitDegrees=10, twistLimitDegrees=10}
 {bone1Name=dyn_chains_r_05, bone2Name=dyn_chains_r_04, swing1LimitDegrees=10, swing2LimitDegrees=10, twistLimitDegrees=10}
 {bone1Name=dyn_chains_l_01, bone2Name=dyn_chains_base, swing1LimitDegrees=10, swing2LimitDegrees=10, twistLimitDegrees=10}
 {bone1Name=dyn_chains_l_02, bone2Name=dyn_chains_l_01, swing1LimitDegrees=10, swing2LimitDegrees=10, twistLimitDegrees=10}
 {bone1Name=dyn_chains_l_03, bone2Name=dyn_chains_l_02, swing1LimitDegrees=10, swing2LimitDegrees=10, twistLimitDegrees=10}
 {bone1Name=dyn_chains_l_04, bone2Name=dyn_chains_l_03, swing1LimitDegrees=10, swing2LimitDegrees=10, twistLimitDegrees=10}
 {bone1Name=dyn_chains_l_05, bone2Name=dyn_chains_l_04, swing1LimitDegrees=10, swing2LimitDegrees=10, twistLimitDegrees=10}
 {
  bone1Name=pack
  bone2Name=root
  swing1LimitDegrees=45
  swing1Motion=Free
  swing2LimitDegrees=45
  swing2Motion=Free
  twistLimitDegrees=45
  twistMotion=Free
 }
]
